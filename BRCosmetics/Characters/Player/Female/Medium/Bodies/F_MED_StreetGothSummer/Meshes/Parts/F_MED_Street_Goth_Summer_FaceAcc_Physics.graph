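# Physics asset for the FaceAcc part of the medium female StreetGothSummer cosmetic body, covering dynamic ponytails and hair strands.
# Capsule bodies cover the head, neck_02, both clavicles and the hair bones.
# Two four-bone ponytail chains and two two-bone strand chains constrain to the head, with tighter swing limits at the roots and zero twist limit.
# Head-to-neck rotation is free, and the clavicle bodies have no constraints.
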PhysicsAsset /BRCosmetics/Characters/Player/Female/Medium/Bodies/F_MED_StreetGothSummer/Meshes/Parts/F_MED_Street_Goth_Summer_FaceAcc_Physics
bodies {
 clavicle_l [{shapeType=Capsule, length=4.75, radius=6.505, shapeName=clavicle_l_capsule, center=<7.316 0 -1.2702>, rotation=<rot -0.0006 -90 -80>}]
 clavicle_r [{shapeType=Capsule, length=5.25, radius=6.005, shapeName=clavicle_r_capsule, center=<-7.3161 0.4822 1.2702>, rotation=<rot -0.0002 -90 100>}]
 dyn_hair_ponytail_l_01 [{
   shapeType=Capsule
   length=0.75
   radius=4
   shapeName=dyn_hair_ponytail_l_01_capsule
   center=<0.2188 0 0.0586>
   rotation=<rot -75.0001 0.0009 -0.0006>
  }]
 dyn_hair_ponytail_l_02 [{
   shapeType=Capsule
   length=1
   radius=3.5
   shapeName=dyn_hair_ponytail_l_02_capsule
   center=<2.8006 -0.2443 0.7504>
   rotation=<rot -85.0003 0.0012 -0.0006>
  }]
 dyn_hair_ponytail_l_03 [{
   shapeType=Capsule
   length=1
   radius=3
   shapeName=dyn_hair_ponytail_l_03_capsule
   center=<2.8478 0.3928 0.2113>
   rotation=<rot -85.0003 -0.0012 0.0018>
  }]
 dyn_hair_ponytail_l_04 [{
   shapeType=Capsule
   length=2.75
   radius=2.75
   shapeName=dyn_hair_ponytail_l_04_capsule
   center=<2.9014 0 0.3741>
   rotation=<rot -80.0002 -0.0003 0.0008>
  }]
 dyn_hair_ponytial_r_01 [{
   shapeType=Capsule
   length=0.75
   radius=4
   shapeName=dyn_hair_ponytial_r_01_capsule
   center=<0.2188 0 0.0586>
   rotation=<rot -75.0001 -0.0009 0.0005>
  }]
 dyn_hair_ponytial_r_02 [{
   shapeType=Capsule
   length=1
   radius=3.5
   shapeName=dyn_hair_ponytial_r_02_capsule
   center=<2.8006 0.0905 0.7504>
   rotation=<rot -85.0003 -0.0011 0.0006>
  }]
 dyn_hair_ponytial_r_03 [{
   shapeType=Capsule
   length=1
   radius=3
   shapeName=dyn_hair_ponytial_r_03_capsule
   center=<2.8478 0.3928 0.2113>
   rotation=<rot -85.0003 0.0013 -0.0018>
  }]
 dyn_hair_ponytial_r_04 [{
   shapeType=Capsule
   length=2.5
   radius=2.25
   shapeName=dyn_hair_ponytial_r_04_capsule
   center=<2.8591 0 0.6141>
   rotation=<rot -80.0003 0.0003 -0.0008>
  }]
 dyn_hair_strand_l_01 [{
   shapeType=Capsule
   length=3.25
   radius=1.5
   shapeName=dyn_hair_strand_l_01_capsule
   center=<3.084 -0.2233 0.056>
   rotation=<rot 84.9618 170.076 175.0378>
  }]
 dyn_hair_strand_l_02 [{
   shapeType=Capsule
   length=3.25
   radius=1.5
   shapeName=dyn_hair_strand_l_02_capsule
   center=<5.0078 -0.2597 -0.4381>
   rotation=<rot 85.0003 -0.0005 -0.0006>
  }]
 dyn_hair_strand_r_01 [{
   shapeType=Capsule
   length=3.25
   radius=1.5
   shapeName=dyn_hair_strand_r_01_capsule
   center=<3.0825 -0.2597 0.0562>
   rotation=<rot 85 179.9998 179.9999>
  }]
 dyn_hair_strand_r_02 [{
   shapeType=Capsule
   length=2.25
   radius=1.5
   shapeName=dyn_hair_strand_r_02_capsule
   center=<4.595 -0.5171 -0.402>
   rotation=<rot 85.0003 0.0006 0.0006>
  }]
 head [{shapeType=Capsule, length=4.75, radius=8.8733, shapeName=head_capsule, center=<5.5645 -0.0766 -0.2066>, rotation=<rot -7.5307 102.1127 -90>}]
 neck_02 [{shapeType=Capsule, length=3.75, radius=4.505, shapeName=neck_02_capsule, rotation=<rot 0 -90.0002 90>}]
}
constraints [
 {
  bone1Name=head
  bone2Name=neck_02
  swing1LimitDegrees=45
  swing1Motion=Free
  swing2LimitDegrees=45
  swing2Motion=Free
  twistLimitDegrees=45
  twistMotion=Free
 }
 {bone1Name=dyn_hair_strand_l_01, bone2Name=head, swing1LimitDegrees=5, swing2LimitDegrees=25, twistLimitDegrees=0}
 {bone1Name=dyn_hair_strand_l_02, bone2Name=dyn_hair_strand_l_01, swing1LimitDegrees=25, swing2LimitDegrees=25, twistLimitDegrees=0}
 {bone1Name=dyn_hair_strand_r_01, bone2Name=head, swing1LimitDegrees=5, swing2LimitDegrees=25, twistLimitDegrees=0}
 {bone1Name=dyn_hair_strand_r_02, bone2Name=dyn_hair_strand_r_01, swing1LimitDegrees=25, swing2LimitDegrees=25, twistLimitDegrees=0}
 {bone1Name=dyn_hair_ponytail_l_01, bone2Name=head, swing1LimitDegrees=1, swing2LimitDegrees=15, twistLimitDegrees=0}
 {bone1Name=dyn_hair_ponytail_l_02, bone2Name=dyn_hair_ponytail_l_01, swing1LimitDegrees=30, swing2LimitDegrees=30, twistLimitDegrees=0}
 {bone1Name=dyn_hair_ponytail_l_03, bone2Name=dyn_hair_ponytail_l_02, swing1LimitDegrees=30, swing2LimitDegrees=30, twistLimitDegrees=0}
 {bone1Name=dyn_hair_ponytail_l_04, bone2Name=dyn_hair_ponytail_l_03, swing1LimitDegrees=30, swing2LimitDegrees=30, twistLimitDegrees=0}
 {bone1Name=dyn_hair_ponytial_r_01, bone2Name=head, swing1LimitDegrees=1, swing2LimitDegrees=15, twistLimitDegrees=0}
 {bone1Name=dyn_hair_ponytial_r_02, bone2Name=dyn_hair_ponytial_r_01, swing1LimitDegrees=30, swing2LimitDegrees=30, twistLimitDegrees=0}
 {bone1Name=dyn_hair_ponytial_r_03, bone2Name=dyn_hair_ponytial_r_02, swing1LimitDegrees=30, swing2LimitDegrees=30, twistLimitDegrees=0}
 {bone1Name=dyn_hair_ponytial_r_04, bone2Name=dyn_hair_ponytial_r_03, swing1LimitDegrees=30, swing2LimitDegrees=30, twistLimitDegrees=0}
]
